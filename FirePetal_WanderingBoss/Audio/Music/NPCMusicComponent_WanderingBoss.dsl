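; NPC music component for the Fire Petal Wandering Boss, driving phased music through the music manager.
; The AI controller's alert changes set a RepNotify variable, so remote instances trigger phase 0 on Unaware and phase 1 on Threatened.
; Health changes trigger phase 2 when health plus shield percent is at most 0.5.
; It also attaches a cylinder audio gameplay volume with enter/exit bindings, and stops music with a death parameter.
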(event EventBeginPlay
  (|CallParentFunction)
  (bind _asbp_player_pawn_fire_petal_wandering_boss (Utilities|Casting|CastToBP_PlayerPawn_FirePetal_WanderingBoss (Components|GetOwner)))
  (bind _output_get (Variables|Default|SetAs%20BP%20Player%20Pawn%20Fire%20Petal%20Wandering%20Boss _asbp_player_pawn_fire_petal_wandering_boss))
  (Default|BindEventtoOnHealthChange _output_get (EventDispatchers|CreateEvent :!event Custom|BossHealthChanged))
  (Variables|Default|SetBossHealth (+ (Utilities|Struct|BreakStructNPCHealthInfo (Variables|/Game/Athena/AI/NPCs/Base/BP_PlayerPawn_NonParticipant_Boss.BP_PlayerPawn_NonParticipant_Boss_C|GetHealthInfo _output_get)) (.ShieldPercent_5_B79C63B64AACF275A1621A9315964DCD (Utilities|Struct|BreakStructNPCHealthInfo (Variables|/Game/Athena/AI/NPCs/Base/BP_PlayerPawn_NonParticipant_Boss.BP_PlayerPawn_NonParticipant_Boss_C|GetHealthInfo _output_get)))))
  (bind _asfort_athena_aibot_controller (Utilities|Casting|CastToFortAthenaAIBotController (Pawn|GetController _output_get)))
  (AlertLevel|BindEventtoOnAlertLevelChangedEvent _asfort_athena_aibot_controller (EventDispatchers|CreateEvent :!event Custom|UpdateMusicState)))

(event EventTick
  (params (DeltaSeconds :type Float%20%28single-precision%29)))

(event Custom|UpdateMusicState
  (params (BotController :type Fort%20Athena%20AIBot%20Controller%20Object%20Reference) (OldAlertLevel :type EAlertLevel%20Enum :default "Unaware") (NewAlertLevel :type EAlertLevel%20Enum :default "Unaware"))
  (Variables|Default|SetNew%20Alert%20Level NewAlertLevel))

(event Custom|UpdateMusicLayer
  (switch Utilities|FlowControl|Switch|SwitchonEAlertLevel (Variables|Default|GetNew%20Alert%20Level)
    (:Unaware
      (Class|NPCMusicManagerComponent|ExecuteMusicTrigger (Variables|Private|GetMusicManager) self (Variables|Default|GetPhase0_TriggerName)))
    (:Threatened
      (Class|NPCMusicManagerComponent|ExecuteMusicTrigger (Variables|Private|GetMusicManager) self (Variables|Default|GetPhase1_TriggerName)))))

(event Custom|BossHealthChanged
  (bind (_healthpercent_2_7edada384c7908ea98debd9958a249de _shieldpercent_5_b79c63b64aacf275a1621a9315964dcd) (Utilities|Struct|BreakStructNPCHealthInfo (Variables|/Game/Athena/AI/NPCs/Base/BP_PlayerPawn_NonParticipant_Boss.BP_PlayerPawn_NonParticipant_Boss_C|GetHealthInfo (Variables|Default|GetAs%20BP%20Player%20Pawn%20Fire%20Petal%20Wandering%20Boss))))
  (bind _output_get (Variables|Default|SetBossHealth (+ _healthpercent_2_7edada384c7908ea98debd9958a249de _shieldpercent_5_b79c63b64aacf275a1621a9315964dcd)))
  (if (<= _output_get 0.5)
    (Class|NPCMusicManagerComponent|ExecuteMusicTrigger (Variables|Private|GetMusicManager) self (Variables|Default|GetPhase2_TriggerName))))

(event Custom|EndMusic
  (Networking|CanExecuteCosmeticEvents
    (:True
      (Class|NPCMusicManagerComponent|SetMusicBoolParam (Variables|Private|GetMusicManager) 0 (Public|GetIsNPCDead) "NPC.State.bIsDead")
      (Class|NPCMusicManagerComponent|ExecuteMusicTrigger (Variables|Private|GetMusicManager) self (Variables|MetaSoundParams|GetOnStopTriggerName)))))

(event EventSpawnAudioShape
  (bind _actorowner (Public|GetActorBase))
  (bind _returnvalue (Game|SpawnActorfromClass (Variables|AGV|GetAGV_CylinderClass) (Transformation|GetActorTransform _actorowner) "Undefined" "SelectDefaultAtRuntime" 0 0))
  (bind _output_get (Variables|AGV|SetAGV_Cylinder _returnvalue))
  (Transformation|AttachActorToActor _output_get _actorowner "None" "SnapToTarget" "SnapToTarget" "KeepRelative" true)
  (Private|AGV|ConfigureAGV))

(fn ConfigureAGV
  (bind _agv_cylinder (Variables|AGV|GetAGV_Cylinder))
  (bind _agv_cylinder_2 (Variables|AGV|GetAGV_Cylinder))
  (bind _audioshapecylinder (Variables|/FortNPCMusic/Audio/AGV_NPCMusicVolume_Cylinder.AGV_NPCMusicVolume_Cylinder_C|GetAudioShapeCylinder _agv_cylinder_2))
  (bind _audiogameplayvolume (Variables|/FortNPCMusic/Audio/AGV_NPCMusicVolume_Cylinder.AGV_NPCMusicVolume_Cylinder_C|GetAudioGameplayVolume _agv_cylinder))
  (bind _volumeradius (Variables|Volume|GetVolumeRadius))
  (Control|SetRadius _audioshapecylinder _volumeradius :!class /Script/AudioShapes.AudioShapeCylinderComponent)
  (Control|SetHalfHeight _audioshapecylinder (Variables|Volume|GetVolumeHalfHeight))
  (Events|BindEventtoOnToggledOn _audiogameplayvolume (EventDispatchers|CreateEvent :!event OnPlayerEnterAGV))
  (Events|BindEventtoOnToggledOff _audiogameplayvolume (EventDispatchers|CreateEvent :!event OnPlayerExitAGV))
  (Class|BPFLNPCMusic|LogNPCMusic (Utilities|Text|FormatText "NSLOCTEXT(\"[46E2315EE59B23359705644EB67A31DC]\", \"F2A0245849455B31BCA7D1B1DA5151A3\", \"AGV {name} created with radius {radius}\")" (Utilities|GetDisplayName _agv_cylinder) _volumeradius)))

(fn OnRep_NewAlertLevel
  (Networking|SwitchHasAuthority
    (:Remote
      (CallFunction|UpdateMusicLayer))))
